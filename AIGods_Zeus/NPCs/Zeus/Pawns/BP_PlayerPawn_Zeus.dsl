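; Player pawn Blueprint for the Zeus boss NPC, handling enlargement, shield collision, health bars and downed state.
; On authority, gameplay-event waits start scaling to phase-three size, track the shield actor and floating tag, and set replicated flags.
; RepNotifies toggle pawn-shield collision ignores, enable health bars and block carrying when downed.
; Remote BeginPlay fires a spawn-in cue and reapplies scale plus a physics-asset refit hack if enlargement already finished.
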(event EventBeginPlay
  (bind _tagstartfloating (Variables|Default|GetTagStartFloating))
  (bind _returnvalue (GameplayTags|MakeGameplayTagContainerfromTag _tagstartfloating))
  (bind _self self)
  ;[1 1 1] Being Play
  (|CallParentFunction)
  (Networking|SwitchHasAuthority
    (:Authority
      ;[1 1 1] Wait for when we should be enlarging.
      (Ability|Async|WaitGameplayEventToActor 0 (Variables|Default|GetTag_StartEnlarging) true
        (:then
          (bind (_asynctaskproxy _payload) (Ability|Async|WaitGameplayEventToActor 0 (Variables|Default|GetTag_ShieldAdded) true)
            (:then
              (Ability|Async|WaitGameplayEventToActor 0 (Variables|Default|GetTag_EndEnlarging) true
                (:then
                  (Ability|Async|WaitGameplayEventToActor 0 _tagstartfloating true
                    (:then
                      (Utilities|FlowControl|Delay 1.0)
                      (Variables|Default|SetCanEnableHealthBars true))
                    (:EventReceived
                      (Ability|Async|WaitGameplayEventToActor 0 (Variables|Default|GetTagEndFloating) true
                        (:then
                          (Ability|GameplayEffect|AddLooseGameplayTags _self _returnvalue true))
                        (:EventReceived
                          (Ability|GameplayEffect|RemoveLooseGameplayTags _self _returnvalue true))))))
                (:EventReceived
                  (Variables|Default|SetRevivingProcessCompleted true))))
            (:EventReceived
              (Variables|Default|SetShieldActor (.OptionalObject (Utilities|Struct|BreakGameplayEventData _payload)))
              (Ability|Async|WaitGameplayEventToActor 0 (Variables|Default|GetTag_ShieldRemoved) true
                (:then
                  (Variables|Default|SetbDisableCollisions true))
                (:EventReceived
                  (Variables|Default|SetbDisableCollisions false))))))
        (:EventReceived
          (CallFunction|StartEnlarge 0.0))))
    (:Remote
      ;[1 1 1] Spawn in Lightning VFX and Show Pawn
      (Utilities|FlowControl|Sequence
        (:then_0
          (GameplayCue|ExecuteGameplayCue%28Burst%29Local self (Variables|Default|GetGC_SpawnIn) (Ability|GameplayCue|MakeGameplayCueParameters :NormalizedMagnitude 0.0 :RawMagnitude 0.0 :Location "0,0,0" :Normal "0,0,0" :GameplayEffectLevel (Variables|Default|GetSpawnInVFXParameter) :AbilityLevel 1 :bReplicateLocationWhenUsingMinimalRepProxy false)))
        (:then_1
          ;[1 1 1] Make sure Pawn is scaled up when it should be
          (if (and (Variables|Default|GetbEnlargedFinished) (Math|Vector|NotEqualExactly(Vector) (Transformation|GetActorScale3D self) (Variables|Default|GetPhase3Scale)))
            (Transformation|SetActorScale3D self (Variables|Default|GetPhase3Scale))
            (Components|SkinnedMesh|SetPhysicsAsset (Variables|Character|GetMesh) "/Game/Characters/Player/Male/Male_Avg_Base/Fortnite_M_Avg_Player_Physics.Fortnite_M_Avg_Player_Physics" true) ; Hack to make sure the Physic Asset fits the size of the AI
          ))))))

(event Custom|StartEnlarge
  (params (EnlargingDuration :type Float%20%28double-precision%29))
  ;[1 1 1] Start Enlarging the Boss (Scale him up)
  (Utilities|FlowControl|DoOnce false
    (:Completed
      (Transformation|SetActorScale3D self (Variables|Default|GetPhase3Scale))
      (Variables|Default|SetbEnlargedFinished true)
      (CallFunction|SetNPCAnimSpeedScalar 0.5)
      (Components|SkinnedMesh|SetPhysicsAsset (Variables|Character|GetMesh) "/Game/Characters/Player/Male/Male_Avg_Base/Fortnite_M_Avg_Player_Physics.Fortnite_M_Avg_Player_Physics" true) ; Hack to make sure the Physic Asset fits the size of the AI
    )))

(event EventOnDBNOStateChanged_Event
  (params (FortPawn :type Fort%20Pawn%20Object%20Reference) (bInIsDBNO :type Boolean))
  ;[1 1 1] Hide Health Bar on DBNO
  (|CallParentFunction FortPawn bInIsDBNO)
  (if bInIsDBNO
    (CallFunction|NPCDisableHealthBars)
    (Variables|Default|SetDBNOBlockCarry true)))

(event Custom|DisableCarry
  ;[1 1 1] Disable Zeus having the Carry Prompt
  (bind _asfort_ability_system_component (Utilities|Casting|CastToFortAbilitySystemComponent (Ability|GetAbilitySystemComponent self)))
  (Tags|BlockAbilitiesWithTags _asfort_ability_system_component (Variables|Default|GetAbilityTagToBlock)))

(fn OnRep_bDisableCollisions
  (bind _shieldactor (Variables|Default|GetShieldActor))
  (if (Variables|Default|GetbDisableCollisions)
    (bind _asactor (Utilities|Casting|CastToActor _shieldactor))
    (Collision|IgnoreActorwhenMoving (Variables|Character|GetCapsuleComponent) _asactor true) ; Make sure to bin on EndPlay to remove the ignore
    (Collision|IgnoreActorwhenMoving (Actor|GetComponentbyClass _asactor "/Script/Engine.SphereComponent") self true) ; Make sure to bin on EndPlay to remove the ignore
    (else
      (bind _asactor_2 (Utilities|Casting|CastToActor _shieldactor))
      (Collision|IgnoreActorwhenMoving (Variables|Character|GetCapsuleComponent) _asactor_2 false) ; Make sure to bin on EndPlay to remove the ignore
      (Collision|IgnoreActorwhenMoving (Actor|GetComponentbyClass _asactor_2 "/Script/Engine.SphereComponent") self false) ; Make sure to bin on EndPlay to remove the ignore
    )))

(fn OnRep_bEnlargedFinished)

(fn OnRep_CanEnableHealthBars
  (if (Variables|Default|GetCanEnableHealthBars)
    (CallFunction|NPCEnableHealthBars)))

(fn OnRep_DBNOBlockCarry
  (if (Variables|Default|GetDBNOBlockCarry)
    (CallFunction|DisableCarry)))

(fn OnRep_ShieldActor)

(fn ConstructionScript
  (|CallParentFunction))
